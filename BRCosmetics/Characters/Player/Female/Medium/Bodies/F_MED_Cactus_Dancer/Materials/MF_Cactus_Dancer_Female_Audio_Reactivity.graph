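# Material function for the Cactus Dancer female outfit that makes masked body areas glow in reaction to music.
# Pre-skinned X position indexes a sampled audio spectrum history; a Custom node blends its four thresholded, weighted bands along Z.
# The shaped level, FX mask green channel and time-of-day strength drive emissive tinted by base color.
# Masked base color also dims toward half brightness when the level is low.
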MaterialFunction /BRCosmetics/Characters/Player/Female/Medium/Bodies/F_MED_Cactus_Dancer/Materials/MF_Cactus_Dancer_Female_Audio_Reactivity
graph {nodes [
  n1 = PreSkinnedLocalBounds()
  n2 = ComponentMask(None=@n1.Half Extents) {r=true}
  AppendVector(A=@n2)
  CollectionParameter() {collection=/Game/Athena/Events/AudioAnalysis/MPC_Audio_Reactive_Parameters, parameterName=Spectrum_Music}
  n3 = FunctionInput() {description=Material Attributes, inputName=MA, inputType=FunctionInput_MaterialAttributes}
  n4 = StaticBool() {value=true}
  n5 = PreSkinnedPosition()
  n6 = Add(A=ComponentMask(None=@n5) {r=true}, B=@n2)
  n7 = Subtract(A=@n1.Max, B=@n1.Max)
  n8 = ComponentMask(None=@n7) {r=true}
  n9 = Multiply(A=VertexInterpolator(VS=Divide(A=@n6, B=@n8) {desc=across}).PS) {constB=0.05, desc=Reduce Time}
  n10 = InverseLinearInterpolate(
   A=ScalarParameter() {defaultValue=0.2, parameterName=AudioReactivityLowerThreshold}
   Value=MaterialFunctionCall!(useSmoothStep=@n4, uvx=@n9) {materialFunction=/Game/Accessories/FORT_Backpacks/Backpack_F_MED_Cactus_Rocker/Materials/SampleAudioMPCHistory}.Interpolated Spectrum
  ) {constA=0.5}
  n11 = MaterialFunctionCall(Color=@n10) {materialFunction=&SplitComponents}
  MaterialExpressionMultiply_5 = Multiply(A=@n11.R, B=ScalarParameter() {defaultValue=1, parameterName=F})
  MaterialExpressionMultiply_6 = Multiply(A=@n11.G, B=ScalarParameter() {defaultValue=1, parameterName=F2})
  MaterialExpressionMultiply_7 = Multiply(A=@n11.B, B=ScalarParameter() {defaultValue=1, parameterName=F3})
  MaterialExpressionMultiply_8 = Multiply(A=ComponentMask(None=@n10) {a=true}, B=ScalarParameter() {defaultValue=1, parameterName=F4})
  n12 = VertexInterpolator(VS=Divide(A=ComponentMask(None=@n5) {b=true}, B=ComponentMask(None=@n7) {b=true}))
  MaterialExpressionOneMinus_2 = OneMinus(None=Power(Base=OneMinus(None=@n12.PS)) {constExponent=1.1})
  n13 = Custom!(
   in1=@MaterialExpressionMultiply_5
   in2=@MaterialExpressionMultiply_6
   in3=@MaterialExpressionMultiply_7
   in4=@MaterialExpressionMultiply_8
   alpha=@MaterialExpressionOneMinus_2
  ) {
   code="float floatValues[4] = {\r\n\t\tin1, \r\n        in2,\r\n\t\tin3, \r\n        in4\r\n};\r\n\r\n\r\n\r\nfloat extendRange = alpha*(4-1);\r\nfloat fractionalPart = frac(extendRange);\r\nuint index = uint(floor(extendRange));\r\n\r\nreturn  lerp(floatValues[index], floatValues[index+1], fractionalPart);"
   description=spreadAcrrossMeshCustom
   outputType=CMOT_Float1
   inputs [
    {
     inputName=in1
     input {expression=&self:MaterialExpressionMultiply_5}
    }
    {
     inputName=in2
     input {expression=&self:MaterialExpressionMultiply_6}
    }
    {
     inputName=in3
     input {expression=&self:MaterialExpressionMultiply_7}
    }
    {
     inputName=in4
     input {expression=&self:MaterialExpressionMultiply_8}
    }
    {
     inputName=alpha
     input {expression=&self:MaterialExpressionOneMinus_2}
    }
   ]
  }
  n14 = Power(Base=Saturate(None=Multiply(A=@n13) {constB=4}), Exp=ScalarParameter() {defaultValue=4, parameterName=EmissivePower}) {constExponent=1.5}
  n15 = TextureObjectParameter() {
   group=Skin FX
   parameterName=SkinFX_Mask
   samplerType=SAMPLERTYPE_Masks
   texture=/BRCosmetics/Characters/Player/Female/Medium/Bodies/F_MED_Cactus_Dancer/Textures/T_F_MED_CactusDancer_Body_FX
  }
  n16 = TextureSample(Tex=@n15)
  n17 = Multiply(A=@n14, B=@n16.G)
  n18 = ScalarParameter() {defaultValue=5, parameterName=EmissiveAudio}
  n19 = ScalarParameter() {defaultValue=3, parameterName=EmissiveAudioNight}
  n20 = MaterialFunctionCall(Value_Morning=@n18, Value_Day=@n18, Value_Evening=@n19, Value_Night=@n19) {materialFunction=&Scalar_Time_of_Day}
  n21 = MaterialFunctionCall!(
   Material=@n3
   Emissive=Multiply(
    A=MaterialFunctionCall!(Material=@n3) {materialFunction=/Engine/Functions/MaterialLayerFunctions/MatLayerBlend_BreakBaseColor}.BaseColor
    B=Multiply(A=@n17, B=@n20.Result) {constB=12}
   )
  ) {materialFunction=/Engine/Functions/MaterialLayerFunctions/MatLayerBlend_Emissive}
  n22 = MaterialFunctionCall(Material=@n21.Blended Material) {materialFunction=&MatLayerBlend_BreakBaseColor}
  n23 = Multiply(A=@n22.BaseColor) {constB=0.5}
  FunctionOutput(None=MaterialFunctionCall!(Material=@n21.Blended Material, NewBaseColor=LinearInterpolate(A=@n23, B=@n22.BaseColor, Alpha=@n14), Mask=@n16.G) {materialFunction=/Engine/Functions/MaterialLayerFunctions/MatLayerBlend_OverrideBaseColor}.Blended Material)
  MaterialFunctionCall!(useSmoothStep=@n4, uvx=@n9) {materialFunction=/Game/Accessories/FORT_Backpacks/Backpack_F_MED_Cactus_Rocker/Materials/SampleAudioMPCInterpolateed}
  MaterialFunctionCall() {materialFunction=&BoundingBoxBased_0-1_UVW}
  NamedRerouteDeclaration(None=@n10) {name=spec, nodeColor=<rgba 1 0 0.1406 1>}
  NamedRerouteUsage()
  VertexInterpolator(VS=Add(A=@n2, B=MaterialFunctionCall() {materialFunction=&BreakOutFloat3Components}.R))
  VertexInterpolator(VS=PreSkinnedPosition())
 ]}
