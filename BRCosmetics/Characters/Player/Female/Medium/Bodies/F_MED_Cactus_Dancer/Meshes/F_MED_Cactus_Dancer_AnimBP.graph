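# Animation Blueprint for the F_MED_Cactus_Dancer female body cosmetic, adding secondary motion to four bracelet bones.
# It copies the attached parent mesh's pose, then runs four chained Anim Dynamics nodes in root-relative space before output.
# Each bracelet slides only along local X under reduced gravity and springs, with angular limits at zero and a slightly different travel range per bone.
# The event graph is empty, and the wearingBackbling variable is unused here.
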AnimBlueprint /BRCosmetics/Characters/Player/Female/Medium/Bodies/F_MED_Cactus_Dancer/Meshes/F_MED_Cactus_Dancer_AnimBP
target_skeleton=&/F_MED_Cactus_Dancer_Skeleton
graphs [
 AnimGraph {nodes [
   n1 = Anim Dynamics - Bone: dyn_bracelet_1(ComponentPose=Local To Component(LocalPose=Copy Pose From Mesh() {bUseAttachedParent=true}), Alpha=1, ExternalForce="0,0,0") {
    angularDampingOverride=1
    angularSpringConstant=20000
    bOverrideLinearDamping=true
    bUseGravityOverride=true
    gravityScale=0.35
    linearDampingOverride=1
    linearSpringConstant=100
    simulationSpace=RootRelative
    alphaBoolBlend {blendOption=Linear}
    boundBone {boneName=dyn_bracelet_1}
    gravityOverride {z=-980}
    physicsBodyDefinitions [{
      boundBone {boneName=dyn_bracelet_1}
      boxExtents=<100 100 100>
      constraintSetup {angularLimitsMax=<0 0 0>, angularLimitsMin=<0 0 0>, angularTarget=<0 0 0>, linearAxesMax=<2.5 0 0>, linearAxesMin=<-3.5 0 0>}
      localJointOffset=<1 0 0>
     }]
    relativeSpaceBone {boneName=upperarm_r}
   }
   n2 = Anim Dynamics - Bone: dyn_bracelet_2(ComponentPose=@n1, Alpha=1, ExternalForce="0,0,0") {
    angularDampingOverride=1
    angularSpringConstant=20000
    bOverrideLinearDamping=true
    bUseGravityOverride=true
    gravityScale=0.35
    linearDampingOverride=1
    linearSpringConstant=100
    simulationSpace=RootRelative
    alphaBoolBlend {blendOption=Linear}
    boundBone {boneName=dyn_bracelet_2}
    gravityOverride {z=-980}
    physicsBodyDefinitions [{
      boundBone {boneName=dyn_bracelet_2}
      boxExtents=<100 100 100>
      constraintSetup {angularLimitsMax=<0 0 0>, angularLimitsMin=<0 0 0>, angularTarget=<0 0 0>, linearAxesMax=<2.5 0 0>, linearAxesMin=<-3 0 0>}
      localJointOffset=<1 0 0>
     }]
    relativeSpaceBone {boneName=upperarm_r}
   }
   n3 = Anim Dynamics - Bone: dyn_bracelet_3(ComponentPose=@n2, Alpha=1, ExternalForce="0,0,0") {
    angularDampingOverride=1
    angularSpringConstant=20000
    bOverrideLinearDamping=true
    bUseGravityOverride=true
    gravityScale=0.35
    linearDampingOverride=1
    linearSpringConstant=100
    simulationSpace=RootRelative
    alphaBoolBlend {blendOption=Linear}
    boundBone {boneName=dyn_bracelet_3}
    gravityOverride {z=-980}
    physicsBodyDefinitions [{
      boundBone {boneName=dyn_bracelet_3}
      boxExtents=<100 100 100>
      constraintSetup {angularLimitsMax=<0 0 0>, angularLimitsMin=<0 0 0>, angularTarget=<0 0 0>, linearAxesMax=<2.5 0 0>, linearAxesMin=<-2.5 0 0>}
      localJointOffset=<1 0 0>
     }]
    relativeSpaceBone {boneName=upperarm_r}
   }
   n4 = Anim Dynamics - Bone: dyn_bracelet_4(ComponentPose=@n3, Alpha=1, ExternalForce="0,0,0") {
    angularDampingOverride=1
    angularSpringConstant=20000
    bOverrideLinearDamping=true
    bUseGravityOverride=true
    gravityScale=0.35
    linearDampingOverride=1
    linearSpringConstant=100
    simulationSpace=RootRelative
    alphaBoolBlend {blendOption=Linear}
    boundBone {boneName=dyn_bracelet_4}
    gravityOverride {z=-980}
    physicsBodyDefinitions [{
      boundBone {boneName=dyn_bracelet_4}
      boxExtents=<100 100 100>
      constraintSetup {angularLimitsMax=<0 0 0>, angularLimitsMin=<0 0 0>, angularTarget=<0 0 0>, linearAxesMax=<2.5 0 0>, linearAxesMin=<-2 0 0>}
      localJointOffset=<1 0 0>
     }]
    relativeSpaceBone {boneName=upperarm_r}
   }
   Output Pose(Result=Component To Local(ComponentPose=@n4)) {layerGroup=None}
  ]}
 EventGraph {links [], nodes []}
]
variables [wearingBackbling {type=number}]
