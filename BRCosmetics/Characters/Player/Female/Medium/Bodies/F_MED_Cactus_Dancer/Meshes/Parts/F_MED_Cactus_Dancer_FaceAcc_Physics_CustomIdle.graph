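# Physics asset for the Cactus Dancer female cosmetic's face accessory, named as a custom-idle variant, covering dynamic lock chains (likely hair).
# Center, front and lettered side chains of capsules hang from the head, asymmetric with an extra left D chain.
# Per-joint swing limits shape their motion, with twist locked or zero-limited.
# Spine, neck, upper-arm and right-clavicle capsules form the body volume, including oversized neck capsules, joined by locked or free angular constraints.
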PhysicsAsset /BRCosmetics/Characters/Player/Female/Medium/Bodies/F_MED_Cactus_Dancer/Meshes/Parts/F_MED_Cactus_Dancer_FaceAcc_Physics_CustomIdle
bodies {
 clavicle_r [{shapeType=Capsule, length=8, radius=6, shapeName=clavicle_r_capsule, center=<-4.8454 0.3934 0.6925>, rotation=<rot 75.1565 0 0>}]
 dyn_C_lock_1 [{shapeType=Capsule, length=5, radius=1.6, shapeName=dyn_C_lock_1_capsule, center=<3 0 0>, rotation=<rot 90 0 0>}]
 dyn_C_lock_2 [{shapeType=Capsule, length=5, radius=1.8, shapeName=dyn_C_lock_2_capsule, center=<4 0 0>, rotation=<rot 90 0 0>}]
 dyn_C_lock_3 [{shapeType=Capsule, length=5, radius=1.8, shapeName=dyn_C_lock_3_capsule, center=<4 0 0>, rotation=<rot 90 0 0>}]
 dyn_C_lock_4 [{shapeType=Capsule, length=5, radius=1.8, shapeName=dyn_C_lock_4_capsule, center=<3 0 0>, rotation=<rot 90 0 0>}]
 dyn_L_frontLock_1 [{shapeType=Capsule, length=4, radius=3, shapeName=dyn_L_frontLock_1_capsule, center=<3 0 0>, rotation=<rot 90 0 0>}]
 dyn_L_frontLock_2 [{shapeType=Capsule, length=4, radius=1, shapeName=dyn_L_frontLock_2_capsule, center=<3 0 0>, rotation=<rot 90 0 0>}]
 dyn_L_frontLock_3 [{shapeType=Capsule, length=4, radius=1, shapeName=dyn_L_frontLock_3_capsule, center=<3 0 0>, rotation=<rot 90 0 0>}]
 dyn_L_frontLock_4 [{shapeType=Capsule, length=6.15, radius=1, shapeName=dyn_L_frontLock_4_capsule, center=<4 0 -0.6>, rotation=<rot 82 0 0>}]
 dyn_L_lock_A_1 [{shapeType=Capsule, length=5, radius=2, shapeName=dyn_L_lock_A_1_capsule, center=<3 0 0>, rotation=<rot 90 0 0>}]
 dyn_L_lock_A_2 [{shapeType=Capsule, length=5, radius=2, shapeName=dyn_L_lock_A_2_capsule, center=<3 0 0>, rotation=<rot 90 0 0>}]
 dyn_L_lock_A_2_5 [{shapeType=Capsule, length=5, radius=1.6, shapeName=dyn_L_lock_A_2_5_capsule, center=<3 0 0>, rotation=<rot 90 0 0>}]
 dyn_L_lock_A_3 [{shapeType=Capsule, length=5, radius=1.6, shapeName=dyn_L_lock_A_3_capsule, center=<3 0 0>, rotation=<rot 90 0 0>}]
 dyn_L_lock_A_4 [{shapeType=Capsule, length=5, radius=1.6, shapeName=dyn_L_lock_A_4_capsule, center=<3 0 0>, rotation=<rot 90 0 0>}]
 dyn_L_lock_A_5 [{shapeType=Capsule, length=5, radius=1.6, shapeName=dyn_L_lock_A_5_capsule, center=<3 0 0>, rotation=<rot 90 0 0>}]
 dyn_L_lock_B_1 [{shapeType=Capsule, length=5, radius=2, shapeName=dyn_L_lock_B_1_capsule, center=<3 0 0>, rotation=<rot 90 0 0>}]
 dyn_L_lock_B_2 [{shapeType=Capsule, length=5, radius=2, shapeName=dyn_L_lock_B_2_capsule, center=<3 0 0>, rotation=<rot 90 0 0>}]
 dyn_L_lock_B_3 [{shapeType=Capsule, length=5, radius=1.6, shapeName=dyn_L_lock_B_3_capsule, center=<3 0 0>, rotation=<rot 90 0 0>}]
 dyn_L_lock_B_4 [{shapeType=Capsule, length=5, radius=1.6, shapeName=dyn_L_lock_B_4_capsule, center=<3 0 0>, rotation=<rot 90 0 0>}]
 dyn_L_lock_C_1 [{shapeType=Capsule, length=5, radius=2, shapeName=dyn_L_lock_C_1_capsule, center=<3 0 0>, rotation=<rot 90 0 0>}]
 dyn_L_lock_C_2 [{shapeType=Capsule, length=5, radius=2, shapeName=dyn_L_lock_C_2_capsule, center=<3 0 0>, rotation=<rot 90 0 0>}]
 dyn_L_lock_C_3 [{shapeType=Capsule, length=5, radius=1.6, shapeName=dyn_L_lock_C_3_capsule, center=<3 0 0>, rotation=<rot 90 0 0>}]
 dyn_L_lock_C_4 [{shapeType=Capsule, length=5, radius=1.6, shapeName=dyn_L_lock_C_4_capsule, center=<3 0 0>, rotation=<rot 90 0 0>}]
 dyn_L_lock_D_1 [{shapeType=Capsule, length=5, radius=1.6, shapeName=dyn_L_lock_D_1_capsule, center=<3 0 0>, rotation=<rot 90 0 0>}]
 dyn_L_lock_D_2 [{shapeType=Capsule, length=5, radius=1.6, shapeName=dyn_L_lock_D_2_capsule, center=<5 0 0>, rotation=<rot 90 0 0>}]
 dyn_L_lock_D_3 [{shapeType=Capsule, length=5, radius=1.6, shapeName=dyn_L_lock_D_3_capsule, center=<4 0 0>, rotation=<rot 90 0 0>}]
 dyn_L_lock_D_4 [{shapeType=Capsule, length=5, radius=1.6, shapeName=dyn_L_lock_D_4_capsule, center=<3 0 0>, rotation=<rot 90 0 0>}]
 dyn_R_frontLock_1 [{shapeType=Capsule, length=4, radius=3, shapeName=dyn_R_frontLock_1_capsule, center=<-3 0 0>, rotation=<rot 90 0 0>}]
 dyn_R_frontLock_2 [{shapeType=Capsule, length=4, radius=1, shapeName=dyn_R_frontLock_2_capsule, center=<-3 0 0>, rotation=<rot 90 0 0>}]
 dyn_R_frontLock_3 [{shapeType=Capsule, length=4, radius=1, shapeName=dyn_R_frontLock_3_capsule, center=<-3 0 0>, rotation=<rot 90 0 0>}]
 dyn_R_frontLock_4 [{shapeType=Capsule, length=7, radius=1, shapeName=dyn_R_frontLock_4_capsule, center=<-4 0 0.4>, rotation=<rot 82 0 0>}]
 dyn_R_lock_A_1 [{shapeType=Capsule, length=5, radius=1, shapeName=dyn_R_lock_A_1_capsule, center=<-3 0 0>, rotation=<rot 90 0 0>}]
 dyn_R_lock_A_2 [{shapeType=Capsule, length=5, radius=1, shapeName=dyn_R_lock_A_2_capsule, center=<-4 0 0>, rotation=<rot 90 0 0>}]
 dyn_R_lock_A_2_5 [{shapeType=Capsule, length=5, radius=1, shapeName=dyn_R_lock_A_2_5_capsule, center=<-4 0 0>, rotation=<rot 90 0 0>}]
 dyn_R_lock_A_3 [{shapeType=Capsule, length=5, radius=1, shapeName=dyn_R_lock_A_3_capsule, center=<-4 0 0>, rotation=<rot 90 0 0>}]
 dyn_R_lock_A_4 [{shapeType=Capsule, length=5, radius=1, shapeName=dyn_R_lock_A_4_capsule, center=<-3 0 0>, rotation=<rot 90 0 0>}]
 dyn_R_lock_A_5 [{shapeType=Capsule, length=5, radius=1, shapeName=dyn_R_lock_A_5_capsule, center=<-3 0 0>, rotation=<rot 90 0 0>}]
 dyn_R_lock_B_1 [{shapeType=Capsule, length=5, radius=1, shapeName=dyn_R_lock_B_1_capsule, center=<-3 0 0>, rotation=<rot 90 0 0>}]
 dyn_R_lock_B_2 [{shapeType=Capsule, length=5, radius=1, shapeName=dyn_R_lock_B_2_capsule, center=<-3 0 0>, rotation=<rot 90 0 0>}]
 dyn_R_lock_B_3 [{shapeType=Capsule, length=5, radius=1, shapeName=dyn_R_lock_B_3_capsule, center=<-3 0 0>, rotation=<rot 90 0 0>}]
 dyn_R_lock_B_4 [{shapeType=Capsule, length=5, radius=1, shapeName=dyn_R_lock_B_4_capsule, center=<-3 0 0>, rotation=<rot 90 0 0>}]
 dyn_R_lock_C_1 [{shapeType=Capsule, length=5, radius=1, shapeName=dyn_R_lock_C_1_capsule, center=<-3 0 0>, rotation=<rot 90 0 0>}]
 dyn_R_lock_C_2 [{shapeType=Capsule, length=5, radius=1, shapeName=dyn_R_lock_C_2_capsule, center=<-3 0 0>, rotation=<rot 90 0 0>}]
 dyn_R_lock_C_3 [{shapeType=Capsule, length=5, radius=1, shapeName=dyn_R_lock_C_3_capsule, center=<-3 0 0>, rotation=<rot 90 0 0>}]
 dyn_R_lock_C_4 [{shapeType=Capsule, length=5, radius=1, shapeName=dyn_R_lock_C_4_capsule, center=<-3 0 0>, rotation=<rot 90 0 0>}]
 head [{shapeType=Capsule, length=1, radius=1, shapeName=head_capsule, rotation=<rot 90 0 0>}]
 neck_01 [{shapeType=Capsule, length=60, radius=3.5, shapeName=neck_01_capsule}]
 neck_02 [{shapeType=Capsule, length=60, radius=2, shapeName=neck_01_capsule, center=<0 -0.5621 0>}]
 spine_02 [{shapeType=Capsule, length=12, radius=8, shapeName=spine_04_capsule, center=<0.2259 0.8107 0>}]
 spine_03 [{shapeType=Capsule, length=12.75, radius=7, shapeName=spine_04_capsule, center=<-0.8829 1.2598 0>}]
 spine_04 [{shapeType=Capsule, length=12, radius=9.5, shapeName=spine_04_capsule, center=<0 1.6532 0>}]
 spine_05 [
  {shapeType=Capsule, length=11.5, radius=7.5, shapeName=spine_05_capsule, center=<0 -0.6342 0>}
  {shapeType=Capsule, length=40, radius=6, shapeName=spine_05_capsule, center=<5.1773 -1.1083 0>}
 ]
 upperarm_l [{shapeType=Capsule, length=20, radius=5, shapeName=upperarm_l_capsule, center=<12 -0.8 0>, rotation=<rot 90 0 0>}]
 upperarm_r [{shapeType=Capsule, length=20, radius=5.5, shapeName=upperarm_r_capsule, center=<-15.3052 -0.0095 0>, rotation=<rot 90 0 0>}]
}
constraints [
 {
  bone1Name=head
  bone2Name=neck_02
  swing1LimitDegrees=45
  swing1Motion=Locked
  swing2LimitDegrees=45
  swing2Motion=Locked
  twistLimitDegrees=1
  twistMotion=Locked
 }
 {
  bone1Name=spine_05
  bone2Name=spine_04
  swing1LimitDegrees=45
  swing1Motion=Locked
  swing2LimitDegrees=45
  swing2Motion=Locked
  twistLimitDegrees=1
  twistMotion=Locked
 }
 {
  bone1Name=upperarm_l
  bone2Name=spine_05
  swing1LimitDegrees=45
  swing1Motion=Locked
  swing2LimitDegrees=45
  swing2Motion=Locked
  twistLimitDegrees=45
  twistMotion=Locked
 }
 {
  bone1Name=upperarm_r
  bone2Name=clavicle_r
  swing1LimitDegrees=45
  swing1Motion=Locked
  swing2LimitDegrees=45
  swing2Motion=Locked
  twistLimitDegrees=45
  twistMotion=Locked
 }
 {bone1Name=dyn_L_frontLock_1, bone2Name=head, swing1LimitDegrees=0, swing2LimitDegrees=45, twistLimitDegrees=0}
 {bone1Name=dyn_L_frontLock_2, bone2Name=dyn_L_frontLock_1, swing1LimitDegrees=0, swing2LimitDegrees=45, twistLimitDegrees=0}
 {bone1Name=dyn_L_frontLock_3, bone2Name=dyn_L_frontLock_2, swing1LimitDegrees=5, swing2LimitDegrees=45, twistLimitDegrees=0}
 {bone1Name=dyn_L_frontLock_4, bone2Name=dyn_L_frontLock_3, swing1LimitDegrees=5, swing2LimitDegrees=45, twistLimitDegrees=0}
 {bone1Name=dyn_R_frontLock_1, bone2Name=head, swing1LimitDegrees=0, swing2LimitDegrees=45, twistLimitDegrees=0}
 {bone1Name=dyn_R_frontLock_2, bone2Name=dyn_R_frontLock_1, swing1LimitDegrees=0, swing2LimitDegrees=45, twistLimitDegrees=0}
 {bone1Name=dyn_R_frontLock_3, bone2Name=dyn_R_frontLock_2, swing1LimitDegrees=5, swing2LimitDegrees=45, twistLimitDegrees=0}
 {bone1Name=dyn_R_frontLock_4, bone2Name=dyn_R_frontLock_3, swing1LimitDegrees=5, swing2LimitDegrees=45, twistLimitDegrees=0}
 {bone1Name=dyn_C_lock_1, bone2Name=head, swing1LimitDegrees=40, swing2LimitDegrees=50, twistLimitDegrees=0, twistMotion=Locked}
 {bone1Name=dyn_C_lock_2, bone2Name=dyn_C_lock_1, swing1LimitDegrees=20, swing2LimitDegrees=40, twistLimitDegrees=45, twistMotion=Locked}
 {bone1Name=dyn_C_lock_3, bone2Name=dyn_C_lock_2, swing1LimitDegrees=20, swing2LimitDegrees=40, twistLimitDegrees=45, twistMotion=Locked}
 {bone1Name=dyn_C_lock_4, bone2Name=dyn_C_lock_3, swing1LimitDegrees=20, swing2LimitDegrees=40, twistLimitDegrees=45, twistMotion=Locked}
 {bone1Name=dyn_L_lock_A_1, bone2Name=head, swing1LimitDegrees=5, swing2LimitDegrees=50, twistLimitDegrees=45, twistMotion=Locked}
 {bone1Name=dyn_L_lock_A_2, bone2Name=dyn_L_lock_A_1, swing1LimitDegrees=10, swing2LimitDegrees=50, twistLimitDegrees=0, twistMotion=Locked}
 {bone1Name=dyn_L_lock_A_2_5, bone2Name=dyn_L_lock_A_2, swing1LimitDegrees=10, swing2LimitDegrees=50, twistLimitDegrees=0, twistMotion=Locked}
 {bone1Name=dyn_L_lock_A_3, bone2Name=dyn_L_lock_A_2_5, swing1LimitDegrees=10, swing2LimitDegrees=50, twistLimitDegrees=0, twistMotion=Locked}
 {bone1Name=dyn_L_lock_A_4, bone2Name=dyn_L_lock_A_3, swing1LimitDegrees=10, swing2LimitDegrees=50, twistLimitDegrees=0, twistMotion=Locked}
 {bone1Name=dyn_L_lock_A_5, bone2Name=dyn_L_lock_A_4, swing1LimitDegrees=60, swing2LimitDegrees=50, twistLimitDegrees=0, twistMotion=Locked}
 {bone1Name=dyn_L_lock_B_1, bone2Name=head, swing1LimitDegrees=20, swing2LimitDegrees=40, twistLimitDegrees=45, twistMotion=Locked}
 {bone1Name=dyn_L_lock_B_2, bone2Name=dyn_L_lock_B_1, swing1LimitDegrees=10, swing2LimitDegrees=10, twistLimitDegrees=45, twistMotion=Locked}
 {bone1Name=dyn_L_lock_B_3, bone2Name=dyn_L_lock_B_2, swing1LimitDegrees=10, swing2LimitDegrees=10, twistLimitDegrees=45, twistMotion=Locked}
 {bone1Name=dyn_L_lock_B_4, bone2Name=dyn_L_lock_B_3, swing1LimitDegrees=60, swing2LimitDegrees=10, twistLimitDegrees=45, twistMotion=Locked}
 {bone1Name=dyn_L_lock_C_1, bone2Name=head, swing1LimitDegrees=40, swing2LimitDegrees=50, twistLimitDegrees=45, twistMotion=Locked}
 {bone1Name=dyn_L_lock_C_2, bone2Name=dyn_L_lock_C_1, swing1LimitDegrees=10, swing2LimitDegrees=50, twistLimitDegrees=45, twistMotion=Locked}
 {bone1Name=dyn_L_lock_C_3, bone2Name=dyn_L_lock_C_2, swing1LimitDegrees=10, swing2LimitDegrees=50, twistLimitDegrees=45, twistMotion=Locked}
 {bone1Name=dyn_L_lock_C_4, bone2Name=dyn_L_lock_C_3, swing1LimitDegrees=60, swing2LimitDegrees=50, twistLimitDegrees=45, twistMotion=Locked}
 {bone1Name=dyn_L_lock_D_1, bone2Name=head, swing1LimitDegrees=40, swing2LimitDegrees=40, twistLimitDegrees=45, twistMotion=Locked}
 {bone1Name=dyn_L_lock_D_2, bone2Name=dyn_L_lock_D_1, swing1LimitDegrees=10, swing2LimitDegrees=50, twistLimitDegrees=45, twistMotion=Locked}
 {bone1Name=dyn_L_lock_D_3, bone2Name=dyn_L_lock_D_2, swing1LimitDegrees=10, swing2LimitDegrees=50, twistLimitDegrees=45, twistMotion=Locked}
 {bone1Name=dyn_L_lock_D_4, bone2Name=dyn_L_lock_D_3, swing1LimitDegrees=60, swing2LimitDegrees=50, twistLimitDegrees=45, twistMotion=Locked}
 {bone1Name=dyn_R_lock_A_1, bone2Name=head, swing1LimitDegrees=5, swing2LimitDegrees=50, twistLimitDegrees=0, twistMotion=Locked}
 {bone1Name=dyn_R_lock_A_2, bone2Name=dyn_R_lock_A_1, swing1LimitDegrees=5, swing2LimitDegrees=50, twistLimitDegrees=0, twistMotion=Locked}
 {bone1Name=dyn_R_lock_A_2_5, bone2Name=dyn_R_lock_A_2, swing1LimitDegrees=5, swing2LimitDegrees=50, twistLimitDegrees=45, twistMotion=Locked}
 {bone1Name=dyn_R_lock_A_3, bone2Name=dyn_R_lock_A_2_5, swing1LimitDegrees=5, swing2LimitDegrees=50, twistLimitDegrees=45, twistMotion=Locked}
 {bone1Name=dyn_R_lock_A_4, bone2Name=dyn_R_lock_A_3, swing1LimitDegrees=5, swing2LimitDegrees=50, twistLimitDegrees=45, twistMotion=Locked}
 {bone1Name=dyn_R_lock_A_5, bone2Name=dyn_R_lock_A_4, swing1LimitDegrees=5, swing2LimitDegrees=50, twistLimitDegrees=45, twistMotion=Locked}
 {bone1Name=dyn_R_lock_B_1, bone2Name=head, swing1LimitDegrees=20, swing2LimitDegrees=40, twistLimitDegrees=45, twistMotion=Locked}
 {bone1Name=dyn_R_lock_B_2, bone2Name=dyn_R_lock_B_1, swing1LimitDegrees=10, swing2LimitDegrees=50, twistLimitDegrees=45, twistMotion=Locked}
 {bone1Name=dyn_R_lock_B_3, bone2Name=dyn_R_lock_B_2, swing1LimitDegrees=10, swing2LimitDegrees=50, twistLimitDegrees=45, twistMotion=Locked}
 {bone1Name=dyn_R_lock_B_4, bone2Name=dyn_R_lock_B_3, swing1LimitDegrees=60, swing2LimitDegrees=50, twistLimitDegrees=45, twistMotion=Locked}
 {bone1Name=dyn_R_lock_C_1, bone2Name=head, swing1LimitDegrees=40, swing2LimitDegrees=50, twistLimitDegrees=45, twistMotion=Locked}
 {bone1Name=dyn_R_lock_C_2, bone2Name=dyn_R_lock_C_1, swing1LimitDegrees=10, swing2LimitDegrees=50, twistLimitDegrees=45, twistMotion=Locked}
 {bone1Name=dyn_R_lock_C_3, bone2Name=dyn_R_lock_C_2, swing1LimitDegrees=10, swing2LimitDegrees=50, twistLimitDegrees=45, twistMotion=Locked}
 {bone1Name=dyn_R_lock_C_4, bone2Name=dyn_R_lock_C_3, swing1LimitDegrees=60, swing2LimitDegrees=50, twistLimitDegrees=45, twistMotion=Locked}
 {
  bone1Name=neck_01
  bone2Name=spine_05
  swing1LimitDegrees=45
  swing1Motion=Free
  swing2LimitDegrees=45
  swing2Motion=Free
  twistLimitDegrees=45
  twistMotion=Free
 }
 {
  bone1Name=neck_02
  bone2Name=neck_01
  swing1LimitDegrees=45
  swing1Motion=Free
  swing2LimitDegrees=45
  swing2Motion=Free
  twistLimitDegrees=45
  twistMotion=Free
 }
 {
  bone1Name=spine_04
  bone2Name=spine_03
  swing1LimitDegrees=45
  swing1Motion=Free
  swing2LimitDegrees=45
  swing2Motion=Free
  twistLimitDegrees=45
  twistMotion=Free
 }
 {
  bone1Name=spine_03
  bone2Name=spine_02
  swing1LimitDegrees=45
  swing1Motion=Free
  swing2LimitDegrees=45
  swing2Motion=Free
  twistLimitDegrees=45
  twistMotion=Free
 }
 {
  bone1Name=clavicle_r
  bone2Name=spine_05
  swing1LimitDegrees=45
  swing1Motion=Free
  swing2LimitDegrees=45
  swing2Motion=Free
  twistLimitDegrees=45
  twistMotion=Free
 }
]
