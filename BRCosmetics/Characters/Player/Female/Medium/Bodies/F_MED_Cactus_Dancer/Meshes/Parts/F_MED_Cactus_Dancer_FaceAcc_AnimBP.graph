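# Post-copy AnimBP for the Cactus Dancer face accessory, adding physics-driven secondary motion to the character part.
# Blends copied parent and owner-mesh poses at the upper arms, then feeds two RigidBody nodes with separate physics assets.
# The neck-based simulation yields to a spine-based one during the custom frontend idle montage outside emotes.
# Update logic selects default-simulation gravity from backbling and frontend state; emote end triggers a delayed dynamics reset.
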AnimBlueprint /BRCosmetics/Characters/Player/Female/Medium/Bodies/F_MED_Cactus_Dancer/Meshes/Parts/F_MED_Cactus_Dancer_FaceAcc_AnimBP
target_skeleton=&/F_MED_Cactus_Dancer_FaceAcc_Skeleton
graphs [
 AnimGraph {nodes [
   n1 = Copy Pose From Mesh(SourceMeshComponent=Get SkelMesh())
   n2 = Copy Pose From Mesh() {bCopyCurves=true, bUseAttachedParent=true}
   n3 = RigidBody!(
    ComponentPose=Local To Component(LocalPose=Layered blend per bone(BasePose=@n2, BlendPoses_0=@n1) {layerSetup [{branchFilters [{boneName=upperarm_l}, {boneName=upperarm_r}]}]})
    bAlphaBoolEnabled:bool=NOT Boolean(A=Get customFrontendIdleOn() -> (customFrontendIdleOn:bool))
    OverrideWorldGravity:Vector=Get gravityFinal() -> (gravityFinal:Vector)
    ExternalForce
   ) -> (Pose) {
    alphaInputType=Bool
    bClampLinearTranslationLimitToRefPose=true
    bOverrideWorldGravity=true
    overridePhysicsAsset=&/F_MED_Cactus_Dancer_FaceAcc_Physics
    simulationSpace=BaseBoneSpace
    alphaBoolBlend {blendOption=Linear}
    baseBoneRef {boneName=neck_01}
    componentAppliedLinearAccClamp=<3000 3000 3000>
    componentLinearAccScale=<0.4 0.4 0.3>
    componentLinearVelScale=<0.4 0.4 0.3>
    overrideWorldGravity {z=-1111}
   }
   n4 = RigidBody(ComponentPose=@n3, bAlphaBoolEnabled=Get customFrontendIdleOn(), OverrideWorldGravity="0,0,-1111") {
    alphaInputType=Bool
    bClampLinearTranslationLimitToRefPose=true
    bOverrideWorldGravity=true
    overridePhysicsAsset=&/F_MED_Cactus_Dancer_FaceAcc_Physics_CustomIdle
    simulationSpace=BaseBoneSpace
    alphaBoolBlend {blendOption=Linear}
    baseBoneRef {boneName=spine_05}
    componentAppliedLinearAccClamp=<3000 3000 3000>
    componentLinearAccScale=<0.4 0.4 0.3>
    componentLinearVelScale=<0.4 0.4 0.3>
    overrideWorldGravity {z=-1111}
   }
   Output Pose(Result=Component To Local(ComponentPose=@n4)) {layerGroup=None}
   Get gravityFinal()
  ]}
 EventGraph {nodes [
   Event BlueprintUpdateAnimation()
   > n1 = Gravity Override State Function DEPRECATED(CharacterPartAnimInstance=Self-Reference(), GameTimeInSeconds=GetGameTimeInSeconds(), GravityOverrideParams=Get Gravity Override Params())
   > Set General Gravity Override(General Gravity Override:Vector=@n1) -> (Output_Get:Vector)
   n2 = AND Boolean(A=CheckForBackbling(), B=NOT Boolean(A=Get bIsInFrontEnd()))
   > Set backblingEquippedGravity(backblingEquippedGravity:Vector=SelectVector(A="0,0,-3000", B="0, 0,-1111", bPickA=@n2)) -> (Output_Get:Vector)
   > n3 = Set gravityFinal(gravityFinal:Vector=SelectVector(
     A=Get backblingEquippedGravity() -> (backblingEquippedGravity:Vector)
     B=Get General Gravity Override() -> (General Gravity Override:Vector)
     bPickA=@n2
    )) -> (Output_Get:Vector)
   n4 = Set SkelMesh(execute=@n7.then, SkelMesh=GetOwnerSkeletalMeshComponent())
   > n5 = Set Anim_Inst(Anim_Inst:Anim Instance Object Reference=GetAnimInstance(self=@n4)) -> (Output_Get:Anim Instance Object Reference)
   Is Valid(exec=@n3.then, InputObject=@n5)
   > Branch(Condition=Get bIsInFrontEnd())
   n6 = Get bIsPlayingEmote()
   > Set customFrontendIdleOn(customFrontendIdleOn:bool=AND Boolean(
     A=Montage_IsPlaying(self=@n5, Montage=/BRCosmetics/Animation/Game/MainPlayer/Menu/BR/CustomSkins/_s20/CactusDancerF/CactusDancer_Female_Idle_Montage)
     B=NOT Boolean(A=@n6)
    )) -> (Output_Get:bool)
   Event OnEndPlayingEmote() {is=title K2Node_Event}
   > Delay(Duration=0.2)
   > ResetDynamics(InTeleportType=TeleportPhysics)
   > ResetDynamics(InTeleportType=ResetPhysics)
   n7 = Event BlueprintInitializeAnimation()
  ]}
]
variables [
 Anim_Inst {type=AnimInstance, default=null}
 backblingEquippedGravity {type=Vector}
 customFrontendIdleOn {type=boolean}
 General Gravity Override {type=Vector}
 Gravity Override Params {type=GravityOverrideParamsStruct, default {front_End {gravityOverride=<0 0 -700>}}}
 gravityFinal {type=Vector}
 SkelMesh {type=SkeletalMeshComponent}
 tiltedUpAlpha {type=number}
 wearingBackbling {type=number}
]
